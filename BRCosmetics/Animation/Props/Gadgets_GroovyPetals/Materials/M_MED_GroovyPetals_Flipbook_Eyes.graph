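# Masked, lit material for the GroovyPetals gadget prop's eyes, showing one of four expressions from a 2x2 flipbook texture.
# Nested threshold checks on scalar pose parameters offset the UVs by half steps, prioritizing blink, then glabella down, then inner squint, then the default frame.
# The texture's alpha drives both opacity mask and base color, alongside roughness and specular parameters.
# The resulting attributes pass through two character PostFX material functions.
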Material /BRCosmetics/Animation/Props/Gadgets_GroovyPetals/Materials/M_MED_GroovyPetals_Flipbook_Eyes
graph {
 properties {
  bAutomaticallySetUsageInEditor=false
  bUseMaterialAttributes=true
  bUsedWithMeshDeformer=true
  bUsedWithMorphTargets=true
  bUsedWithSkeletalMesh=true
  blendMode=BLEND_Masked
  pixelDepthOffsetMode=PDOM_Legacy
  refractionMethod=RM_IndexOfRefraction
  shadingModel=MSM_FromMaterialExpression
  subsurfaceProfile=&SS_HeroSkin_02
  usedShadingModels=Default Lit
 }
 nodes [
  n1 = ScalarParameter() {group=GroovyPetalEyes, parameterName=R_blink_pose, sortPriority=12}
  n2 = Constant2Vector() {g=0.5, r=0.5}
  n3 = Constant2Vector() {g=0.5}
  n4 = Constant2Vector() {r=0.5}
  n5 = If(
   A=ScalarParameter() {group=GroovyPetalEyes, parameterName=C_glabella_down_pose, sortPriority=11}
   "A > B"=@n3
   "A == B"=@n3
   "A < B"=If(
    A=ScalarParameter() {group=GroovyPetalEyes, parameterName=R_squint_inner_pose, sortPriority=10}
    "A > B"=@n4
    "A == B"=@n4
    "A < B"=Constant2Vector()
   ) {constB=0.5}
  ) {constB=0.5}
  n6 = TextureSampleParameter2D(UVs=Add(A=TextureCoordinate(), B=If(A=@n1, "A > B"=@n2, "A == B"=@n2, "A < B"=@n5) {constB=0.5})) {
   group=GroovyPetalEyes
   parameterName=GroovyPetalFlipbook
   sortPriority=-1
   texture=/BRCosmetics/Animation/Props/Gadgets_GroovyPetals/Textures/Emote_Prop_GroovyPedals_Flipbook_Eyes
  }
  n7 = SetMaterialAttributes!(
   MaterialAttributes
   Opacity Mask=@n6.A
   Base Color=@n6.A
   Roughness=ScalarParameter() {defaultValue=0.5, group=GroovyPetalEyes, parameterName=Roughness, sortPriority=0}
   Specular=ScalarParameter() {defaultValue=0.25, group=GroovyPetalEyes, parameterName=Specular, sortPriority=1}
  ) {attributeSetTypes [
    #1
    #2
    #3
    #4
   ]}
  n8 = MaterialFunctionCall(Material=@n7) {materialFunction=&BaseFN_CharacterFunction_PostFX}
  MaterialFunctionCall(In=@n8.Result) {materialFunction=&CINE_MF_CharPostFX}
 ]
}
